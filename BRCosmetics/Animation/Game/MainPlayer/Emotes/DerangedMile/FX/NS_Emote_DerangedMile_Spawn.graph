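# Niagara spawn effect for the DerangedMile emote, combining mesh Film streaks with a burst of star sprites.
# Film bursts short-lived partitioned meshes that stretch out along X while curves animate their material parameters.
# Stars001 throws spinning, warm-tinted sprites outward under drag and upward gravity.
# Front-end modifiers lower opacity outside gameplay; the AirglowLWE and SmokeLWE emitters show no modules or renderers here.
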NiagaraSystem /BRCosmetics/Animation/Game/MainPlayer/Emotes/DerangedMile/FX/NS_Emote_DerangedMile_Spawn
dependencies {
 usedDataInterfaces [&NiagaraDataInterfaceArrayFloat, &NiagaraDataInterfaceCurve, &NiagaraDataInterfaceArrayFloat3]
 usedDynamicInputs []
 usedModules [&SystemState]
}
emitters {
 AirglowLWE {rendererClasses [], renderers []}
 Film {
  emitterUpdateScript [
   &EmitterState(Life Cycle Mode=::NewEnumerator1, Loop Behavior=::NewEnumerator1, Loop Duration Mode=::NewEnumerator1, UseLoopDelay=false)
   &SpawnBurst_Instantaneous(Spawn Count=6, Use Spawn Probability=false, UseLoopCountLimit=false)
  ]
  particleSpawnScript [
   &PartitionParticles()
   &InitializeParticle(
    Color=<rgba 0 0 0 4>
    Lifetime=&RandomRangeFloat(Minimum=0.2, Maximum=0.8)
    Mesh Scale=<3 1 1>
    Mesh Scale Mode=::NewEnumerator4
    UsePositionOffset=false
    Write Material Random=true
   )
   &InitialMeshOrientation(Rotation=<0 -0.25 0>, Use Rotation Vector=true)
  ]
  particleUpdateScript [
   &ParticleState()
   &ScaleMeshSize(Scale Factor=&Multiply_Vector(
     A=&Multiply_Vector(
      A=&SelectVectorFromArray(
       Direct Array Index=$Particles.PartitionParticles.Partition
       Vector Selection Array=DataInterface() {propertyValues {FloatData [<0.95 1 1>, <1 1 1>]}}
      )
      B=<0.4 0.4 0.4>
     )
     B=&MakeVector(
      X=&FloatFromCurve(FloatCurve=DataInterface() {propertyValues {Curve {
          defaultValue=1
          keys [{arriveTangent=6.5359, interpMode=RCIM_Cubic, leaveTangent=6.5359, tangentMode=RCTM_User, time=0, value=0}, <at 0.3367 1 RCIM_Cubic>]
         }}})
      Y=1
      Z=1
     )
    ))
   &FN_Mesh_Attributes(
    Color=&MakeLinearColorFromVectorAndFloat(
     "Vector (RGB)"=&Multiply_Vector(A=&MakeVectorFromLinearColorRGB(Color=<rgba 1.5 1.4813 1.4665 1>), B=<1 1 1>)
     "Float (Alpha)"=&Multiply_Float(A=&FN_FrontEndFloatModifier(Gameplay=1, Front End=0.6), B=0.6)
    )
    Create Color Attribute=true
    Write to Module Namespace=true
   )
   &DynamicMaterialParameters(
    Index 0 Param 1=1.5
    Index 0 Param 2=&Multiply_Float(A=6, B=&FloatFromCurve(FloatCurve=DataInterface() {propertyValues {Curve {keys [<at 0 1>,<at 1 0>]}}}))
    Index 0 Param 3=&SelectFloatFromArray(Direct Array Index=$Particles.PartitionParticles.Partition, Float Selection Array=DataInterface() {propertyValues {FloatData [0.2,0.8]}})
    Index 0 Param 4=&FloatFromCurve(FloatCurve=DataInterface() {propertyValues {Curve {keys [<at 0 1>,<at 1 50>]}}})
    Index 1 Param 1=&FloatFromCurve(FloatCurve=DataInterface() {propertyValues {Curve {keys [<at 0 -1>,<at 0.3 0>]}}})
    Index 1 Param 2=&FloatFromCurve(FloatCurve=DataInterface() {propertyValues {Curve {keys [<at 0.4 0>,<at 1 -1>]}}})
    Module.Index0Float0Write=true
    Module.Index0Float1Write=true
    Module.Index0Float2Write=true
    Module.Index0Float3Write=true
    Module.Index1Float0Write=true
    Module.Index1Float1Write=true
    Module.Index1Float2Write=true
    Module.Index1Float3Write=true
    Write Parameter Index 1=::NewEnumerator1
   )
   &ScaleColor(Scale Alpha=&FN_FrontEndFloatModifier(Gameplay=1, Front End=0.6), ScaleA=true, ScaleRGB=true)
  ]
  rendererClasses [&NiagaraMeshRendererProperties]
  renderers [
   {rendererClass=&NiagaraMeshRendererProperties, rendererIndex=0}
   {rendererClass=&NiagaraMeshRendererProperties, rendererIndex=1}
  ]
 }
 SmokeLWE {rendererClasses [], renderers []}
 Stars001 {
  emitterUpdateScript [
   &EmitterState()
   &SpawnBurst_Instantaneous(Spawn Count=20, Use Spawn Probability=false, UseLoopCountLimit=false)
  ]
  particleSpawnScript [
   &PartitionParticles()
   /SetVariables(Particles.SizeDetermination:NiagaraFloat=&UniformAOrBFloat(A=0, Distribution Weight=0.7, Evaluation Type=::NewEnumerator0, Recalculate Random Each Loop=true, Override Seed=false)) {isSetParameters=true}
   &InitializeParticle(
    Color=<rgba 4 2 0 1>
    Lifetime Max=0.9
    Lifetime Min=0.3
    Lifetime Mode=::NewEnumerator1
    Sprite Rotation Mode=::NewEnumerator2
    Sprite Size Mode=::NewEnumerator3
    Uniform Sprite Size=&SelectFloatFromArray(
     Direct Array Index=&MakeIntFromFloat(Float=$Particles.SizeDetermination)
     Float Selection Array=DataInterface() {propertyValues {FloatData [16,6]}}
    )
    UsePositionOffset=true
    Write Material Random=true
   )
   &ShapeLocation(
    MaskSpawnGroup=false
    Override Seed=false
    Shape Origin=/Niagara/DynamicInputs/Helpers/SimulationPosition()
    Sphere Radius=1
    UseOwnerScale=true
   )
   &AddVelocity(
    Module.Use Falloff Distance=false
    Override Seed=false
    Velocity Mode=::NewEnumerator1
    Velocity Speed=&RandomRangeFloat(Minimum=150, Maximum=400)
   )
  ]
  particleUpdateScript [
   /SetVariables(Particles.RandomSize:NiagaraFloat=&Multiply_Float(
     A=&RandomRangeFloat(Minimum=1.1, Maximum=1.3)
     B=&FloatFromCurve(
      FloatCurve=DataInterface() {propertyValues {Curve {
         defaultValue=1
         keys [
          <at 0 0.8 RCIM_Cubic>
          {arriveTangent=0.2815, interpMode=RCIM_Cubic, leaveTangent=0.2815, time=0.2397, value=0.8835}
          {arriveTangent=1.4685, interpMode=RCIM_Cubic, leaveTangent=1.4685, time=0.7104, value=1}
          {tangentMode=RCTM_User, time=1, value=2}
         ]
        }}}
      Scale Curve=&Multiply_Float(
       A=0.8
       B=&FloatFromCurve(
        FloatCurve=DataInterface() {propertyValues {Curve {
           defaultValue=1
           keys [{interpMode=RCIM_Cubic, tangentMode=RCTM_User, time=0, value=1}, {interpMode=RCIM_Cubic, tangentMode=RCTM_User, time=1, value=0}]
          }}}
        Scale Curve=2
       )
      )
     )
    )) {isSetParameters=true}
   /SetVariables(Particles.AlphaAll:NiagaraFloat=&FloatFromCurve(
     FloatCurve=DataInterface() {propertyValues {Curve {defaultValue=1, keys [{tangentMode=RCTM_User, time=0.1, value=1}, {tangentMode=RCTM_User, time=0.4, value=1}]}}}
     Scale Curve=1.1
    )) {isSetParameters=true}
   &ParticleState()
   &SpriteRotationRate(Rotation Rate=&RandomRangeFloat(Minimum=-300, Maximum=300))
   &ScaleSpriteSize(Scale Sprite Size Mode=::NewEnumerator0, Uniform Scale Factor=&Multiply_Float(A=$Particles.RandomSize, B=1))
   &Drag(Drag=5, Use Linear Drag=true, Use Rotational Drag=true)
   &GravityForce(Gravity=<0 0 200>)
   &SolveForcesAndVelocity(Clamp Velocity=false, Limit Acceleration=false)
   &DynamicMaterialParameters(
    Index 1 Param 1=3
    Index 1 Param 2=3
    Module.Index0Float0Write=true
    Module.Index0Float1Write=true
    Module.Index0Float2Write=true
    Module.Index0Float3Write=true
    Module.Index1Float0Write=true
    Module.Index1Float1Write=true
    Module.Index1Float2Write=true
    Module.Index1Float3Write=true
    Write Parameter Index 1=::NewEnumerator1
   )
   &FN_Sprite_Attributes(
    Color=&MakeLinearColorFromVectorAndFloat(
     "Vector (RGB)"=&Multiply_Vector(A=&MakeVectorFromLinearColorRGB(Color=<rgba 1.5 0.825 0.15 1>), B=<1 1 1>)
     "Float (Alpha)"=&Multiply_Float(
      A=&FloatFromCurve(
       FloatCurve=DataInterface() {propertyValues {Curve {
          defaultValue=1
          keys [
           {interpMode=RCIM_Cubic, tangentMode=RCTM_User, time=0, value=0}
           {arriveTangent=1.6667, interpMode=RCIM_Cubic, leaveTangent=1.6667, time=0.2, value=1}
           {interpMode=RCIM_Cubic, tangentMode=RCTM_User, time=0.6, value=1}
           {tangentMode=RCTM_User, time=1, value=0}
          ]
         }}}
       Scale Curve=$Particles.AlphaAll
      )
      B=&Multiply_Float(
       A=&FN_FrontEndFloatModifier(Gameplay=0.3, Front End=0.1)
       B=&FNTimeOfDayScalar(Night Scale=1, Morning Scale=1, Evening Scale=1)
      )
     )
    )
    Create Color Attribute=true
    Create SpriteSize Attribute=true
    SpriteSize=&Multiply_Vector2D_ByFloat(Float=&Multiply_Float(A=20, B=$Particles.RandomSize), Vector2D=<1 1>)
    Write to Module Namespace=true
   )
   &ScaleColor(
    Scale Alpha=&Multiply_Float(
     A=&FloatFromCurve(
      FloatCurve=DataInterface() {propertyValues {Curve {
         defaultValue=1
         keys [
          {interpMode=RCIM_Cubic, tangentMode=RCTM_User, time=0, value=1}
          <at 0.2 1 RCIM_Cubic>
          {interpMode=RCIM_Cubic, tangentMode=RCTM_User, time=0.6, value=1}
          {tangentMode=RCTM_User, time=1, value=1}
         ]
        }}}
      Scale Curve=$Particles.AlphaAll
     )
     B=0.5
    )
    ScaleA=true
    ScaleRGB=true
   )
  ]
  rendererClasses [&NiagaraSpriteRendererProperties]
  renderers [
   {rendererClass=&NiagaraSpriteRendererProperties}
   {rendererClass=&NiagaraSpriteRendererProperties, rendererIndex=1}
  ]
 }
}
summary {
 systemName=NS_Emote_DerangedMile_Spawn
 emitters [
  {emitterName=AirglowLWE, rendererClasses []}
  {emitterName=SmokeLWE, rendererClasses []}
  {emitterName=Stars001, rendererClasses [&NiagaraSpriteRendererProperties]}
  {emitterName=Film, rendererClasses [&NiagaraMeshRendererProperties]}
 ]
 userVariables [
  User.bIsFrontEnd {type=NiagaraBool}
  User.bIsFrontEndPreview {type=NiagaraBool}
  User.NormalizedNotifyProgress {type=NiagaraFloat, defaultValue=0, description=""}
  User.ParaGlideLeanAlpha {type=NiagaraFloat, defaultValue=1, description=""}
 ]
}
